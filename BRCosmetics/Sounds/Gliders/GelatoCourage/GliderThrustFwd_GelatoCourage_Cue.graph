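# Sound cue for the forward thrust of the GelatoCourage cosmetic glider.
# A random node picks one of three equally weighted thrust waves, and a modulator adds slight pitch and volume variation.
# The result feeds a two-input mixer whose second input is empty.
# Playback is at reduced volume, using shared glider thrust attenuation and concurrency, the glider landing sound class and a glider-loop audio properties sheet.
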SoundCue /BRCosmetics/Sounds/Gliders/GelatoCourage/GliderThrustFwd_GelatoCourage_Cue
root=Mixer_0
nodes [
 {id=Mixer_0, children [Modulator_0,null], properties {inputVolume [1,1]}}
 {id=Modulator_0, children [Random_0], properties {pitchMax=1.1, pitchMin=0.9, volumeMax=1.02, volumeMin=0.98}}
 {id=Random_0, children [WavePlayer_0,WavePlayer_1,WavePlayer_2], properties {weights [1,1,1]}}
 {id=WavePlayer_0, properties {soundWaveAssetPtr=&/IP_Glider_GelatoCourage_FwdThrust_01}}
 {id=WavePlayer_1, properties {soundWaveAssetPtr=&/IP_Glider_GelatoCourage_FwdThrust_03}}
 {id=WavePlayer_2, properties {soundWaveAssetPtr=&/IP_Glider_GelatoCourage_FwdThrust_02}}
]
properties {
 attenuationSettings=&Fort_Player_Glider_Thrust_Athena_Att
 audioPropertiesSheet=&APS_Cosmetic_Glider_Loop
 maxDistance=12500
 soundClassObject=&Fort_Gliders_Land
 volumeMultiplier=0.5
 concurrencySet [&Athena_Glider_Thrust_Conc]
}
