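# Looping thrust sound cue for the GelatoCourage glider, attenuated with the player glider thrust settings.
# A mixer blends four looping layers (mechanical, low, engine idle and forward engine), each driven by continuous modulators.
# AnyDirection raises pitch on the first three and volume on mechanical and low, while inversely ducking idle.
# Forwards drives forward-engine volume and pitch, with an enveloper adding a volume fade-in.
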SoundCue /BRCosmetics/Sounds/Gliders/GelatoCourage/GliderThrustLoop_GelatoCourage_Cue
root=Mixer_0
nodes [
 {
  id=Enveloper_3
  children [WavePlayer_10]
  properties {pitchCurve {editorCurveData {keys [{time=0}]}}, volumeCurve {editorCurveData {keys [<at 0 0>,<at 1 1>]}}}
 }
 {
  id=Mixer_0
  children [ModulatorContinuous_3, ModulatorContinuous_2, ModulatorContinuous_6, ModulatorContinuous_5, null]
  properties {inputVolume [0.4,0.8,0.8,1,0.8]}
 }
 {
  id=ModulatorContinuous_2
  children [WavePlayer_2]
  properties {
   pitchModulationParams {default=0, maxOutput=1.1, minOutput=1, parameterName=AnyDirection}
   volumeModulationParams {default=0, maxOutput=0.8, minOutput=0.3, parameterName=AnyDirection}
  }
 }
 {
  id=ModulatorContinuous_3
  children [WavePlayer_3]
  properties {
   pitchModulationParams {default=0, maxOutput=1.1, minOutput=1, parameterName=AnyDirection}
   volumeModulationParams {default=0, maxOutput=0.7, minOutput=0.3, parameterName=AnyDirection}
  }
 }
 {
  id=ModulatorContinuous_5
  children [WavePlayer_5]
  properties {
   pitchModulationParams {default=0, maxOutput=1.1, minOutput=1, parameterName=AnyDirection}
   volumeModulationParams {default=0, maxOutput=0.1, minOutput=0.6, parameterName=AnyDirection}
  }
 }
 {
  id=ModulatorContinuous_6
  children [Enveloper_3]
  properties {
   pitchModulationParams {default=0, minOutput=0.98, parameterName=Forwards}
   volumeModulationParams {default=0, maxOutput=0.7, parameterName=Forwards}
  }
 }
 {
  id=WavePlayer_10
  properties {bLooping=true, soundWaveAssetPtr=&/IP_Glider_GelatoCourage_Thrust_LP_EngineFwd}
 }
 {
  id=WavePlayer_2
  properties {bLooping=true, soundWaveAssetPtr=&/IP_Glider_GelatoCourage_Thrust_LP_Low}
 }
 {
  id=WavePlayer_3
  properties {bLooping=true, soundWaveAssetPtr=&/IP_Glider_GelatoCourage_Thrust_LP_Mech}
 }
 {
  id=WavePlayer_5
  properties {bLooping=true, soundWaveAssetPtr=&/IP_Glider_GelatoCourage_Thrust_LP_EngineIdle}
 }
]
properties {
 attenuationSettings=&Fort_Player_Glider_Thrust_Athena_Att
 audioPropertiesSheet=&APS_Cosmetic_Glider_Loop
 maxDistance=12500
 soundClassObject=&Fort_Gliders_Land
 volumeMultiplier=0.9
 concurrencySet [null]
}
